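# Niagara effect for the FareSporeMilk emote that spawns glowing circuit-board bursts on animation beats.
# One emitter spawns when User.AnimBeat or AnimBeat2 passes a threshold, using last-frame HasSpawned flags as edge detection, with size picked from an array by emitter age.
# A random index selects matched palette colors per particle.
# Each particle drives three sprite renderers, a mesh and a light, animating circuit, wave, glow and dark layers on separate normalized ages.
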NiagaraSystem /BRCosmetics/Animation/Game/MainPlayer/Emotes/FareSporeMilk/FX/NS_Emote_FareSporeMilk_Circuit
dependencies {
 usedDataInterfaces [
  &NiagaraDataInterfaceArrayFloat
  &NiagaraDataInterfaceArrayColor
  &NiagaraDataInterfaceColorCurve
  &NiagaraDataInterfaceCurve
 ]
 usedDynamicInputs [&Vector4FromCurve, /Niagara/DynamicInputs/Integer/V2/Round_Float]
 usedModules [
  &SystemState
  &self:SystemSpawnScript.SystemScriptSource.SystemScriptGraph.NiagaraNodeAssignment_3.SetVariables_#1
  /Game/Effects/Niagara/Modules/Logic/UpdateValueStateChange
  &self:SystemSpawnScript.SystemScriptSource.SystemScriptGraph.NiagaraNodeAssignment_0.SetVariables_#2
 ]
}
emitters {Circuitboard {
  emitterSpawnScript [/SetVariables() {isSetParameters=true}]
  emitterUpdateScript [
   &EmitterState(
    Inactive Response=::NewEnumerator1
    Life Cycle Mode=::NewEnumerator1
    Loop Behavior=::NewEnumerator1
    Loop Duration Mode=::NewEnumerator1
    UseLoopDelay=false
   )
   &SpawnPerFrame(
    Spawn:NiagaraBool=&Boolean_AndOperation(
     Bool1=&Boolean_OrOperation(
      Bool1=&SetBoolByFloatComparison(A=$User.AnimBeat, B=0.1)
      Bool2=&SetBoolByFloatComparison(A=$User.AnimBeat2, B=0.1)
     )
     Bool2=&Boolean_NotOperation(Bool=&Boolean_OrOperation(Bool1=$Emitter.HasSpawned, Bool2=$Emitter.HasSpawned2))
    )
    Use Spawn Probability=false
   )
   /SetVariables(Emitter.Size:NiagaraFloat=&SelectFloatFromArray(
     Direct Array Index=&MakeIntFromFloat(Float=&Multiply_Float(A=$Emitter.Age, B=10))
     Interpolation=/Niagara/DynamicInputs/Math/Floor_Float(Float:NiagaraFloat=$Emitter.Age)
     Direct Array Mode=::NewEnumerator1
     Float Selection Array=DataInterface() {propertyValues {FloatData [0.35,0.5,0.4,0.24]}}
    )) {isSetParameters=true}
   /SetVariables(
    Emitter.HasSpawned:NiagaraBool=&SetBoolByFloatComparison(A=$User.AnimBeat, B=0.1)
    Emitter.HasSpawned2:NiagaraBool=&SetBoolByFloatComparison(A=$User.AnimBeat2, B=0.1)
   ) {isSetParameters=true}
  ]
  particleSpawnScript [
   /SetVariables(Particles.RandomSeed:NiagaraInt32=&RandomRangeInt(Minimum=0, Maximum=3, Randomness Mode=::NewEnumerator1, Random Seed=4)) {isSetParameters=true}
   /SetVariables(
    Particles.BackgroundColor:LinearColor=&SelectLinearColorFromArray(
     Array Sampling Mode=::NewEnumerator1
     Direct Array Index=$Particles.RandomSeed
     Direct Array Mode=::NewEnumerator1
     Evaluation Type=::NewEnumerator1
     Interpolation=$Emitter.Age
     Color Selection Array=DataInterface() {propertyValues {ColorData [<rgba 0.3556 0 0.467 1>, <rgba 0.1334 0.0055 0.1753 1>, <rgba 0.7622 0.0518 1 1>, <rgba 0.2248 0 0.467 1>]}}
     Random Seed=?
    )
    Particles.DarkBackgroundColor:LinearColor=&SelectLinearColorFromArray(
     Array Sampling Mode=::NewEnumerator1
     Direct Array Index=$Particles.RandomSeed
     Direct Array Mode=::NewEnumerator1
     Evaluation Type=::NewEnumerator1
     Interpolation=$Emitter.Age
     Color Selection Array=DataInterface() {propertyValues {ColorData [<rgba 0.3182 0.0794 0.7587 1>, <rgba 0.8189 0.1122 0.4814 5>, <rgba 0.1224 0.3202 0.8652 5>, <rgba 0.3182 0.0794 0.7587 1>]}}
     Random Seed=?
    )
    Particles.MainColor:LinearColor=&SelectLinearColorFromArray(
     Array Sampling Mode=::NewEnumerator1
     Direct Array Index=$Particles.RandomSeed
     Direct Array Mode=::NewEnumerator1
     Evaluation Type=::NewEnumerator1
     Interpolation=$Emitter.Age
     Color Selection Array=DataInterface() {propertyValues {ColorData [<rgba 0.7835 0.0924 1 1>, <rgba 1 0.7849 0.1747 1>, <rgba 0.0532 0.7534 1 1>, <rgba 0.7835 0.0924 1 1>]}}
     Random Seed=?
    )
   ) {isSetParameters=true}
   &InitializeParticle(
    Color=<rgba 10 10 10 1>
    Mesh Scale Mode=::NewEnumerator3
    Mesh Uniform Scale=&Multiply_Float(A=0.5, B=$Emitter.Size)
    Sprite Rotation Angle=&SelectFloatFromArray(
     Array Sampling Mode=::NewEnumerator0
     Direct Array Index=?
     Float Selection Array=DataInterface() {propertyValues {FloatData [0,0.25,0.5,0.75]}}
     Random Seed=0
    )
    Sprite Rotation Mode=::NewEnumerator1
    Sprite Size Mode=::NewEnumerator3
    Uniform Sprite Size=&Multiply_Float(A=157, B=$Emitter.Size)
    UsePositionOffset=true
    Write Material Random=true
   )
   &ShapeLocation(Large Radius=70, MaskSpawnGroup=false, Override Seed=false, Shape Primitive=::NewEnumerator3, UseOwnerScale=true)
   /SetVariables(
    Particles.Wave Color:LinearColor=&Multiply_LinearColorByFloat(Linear Color=$Particles.BackgroundColor, Float=2)
    Particles.Wave Lifetime=1.2
   ) {isSetParameters=true}
   /SetVariables(
    Particles.Glow Lifetime=0.4
    Particles.Glow Size=&Multiply_Vector2D_ByFloat(Float=$Emitter.Size, Vector2D=&Multiply_Vector2D_ByFloat(Float=35, Vector2D=<1 1>))
   ) {isSetParameters=true}
   /SetVariables(Particles.Circuit LIfetime=1.2) {isSetParameters=true}
   &InitialMeshOrientation(Mesh Orientation Mode=::NewEnumerator2, Orientation Vector=<0 0 1>, Use Rotation Vector=false)
   &CameraOffset(Camera Offset Amount=10)
   /SetVariables(Particles.Dark Color:LinearColor=$Particles.DarkBackgroundColor, Particles.Dark Lifetime=1, Particles.Dark Size:Vector2f=<80 80>) {isSetParameters=true}
  ]
  particleUpdateScript [
   &ParticleState(Kill Particles When Lifetime Has Elapsed=false, Let Infinitely Lived Particles Die When Emitter Deactivates=true)
   /SetVariables(
    Particles.Update Circuit Age:NiagaraFloat=&Divide_Float(A=$Particles.Age, B=$Particles.Circuit LIfetime)
    Particles.Update Dark Age:NiagaraFloat=&Divide_Float(A=$Particles.Age, B=$Particles.Dark Lifetime)
    Particles.Update Glow Age:NiagaraFloat=&Divide_Float(A=$Particles.Age, B=$Particles.Glow Lifetime)
    Particles.Update Wave Age:NiagaraFloat=&Divide_Float(A=$Particles.Age, B=$Particles.Wave Lifetime)
   ) {isSetParameters=true}
   &ScaleColor(
    Scale Alpha=&Multiply_Float(
     A=&FNTimeOfDayScalar(Night Scale=0.5, Morning Scale=0.5, Evening Scale=0.4)
     B=&FloatFromCurve(FloatCurve=DataInterface() {propertyValues {Curve {keys [<at 0 0>,<at 0.1 1>,<at 0.8 1>,<at 1 0>]}}}, Scale Curve=5)
    )
    Scale RGB=&MakeVectorFromLinearColorRGB(Color=&Multiply_LinearColorByFloat(Linear Color=$Particles.MainColor, Float=3))
    ScaleA=true
    ScaleRGB=true
   )
   &DynamicMaterialParameters(
    Index 0 Param 1=&FloatFromCurve(
     FloatCurve=DataInterface() {propertyValues {Curve {
        defaultValue=1
        keys [{arriveTangent=1.422, interpMode=RCIM_Cubic, leaveTangent=1.422, tangentMode=RCTM_User, time=0, value=0}, <at 0.837 1 RCIM_Cubic>]
       }}}
     CurveIndex=$Particles.Update Circuit Age
    )
    Index 0 Param 2=&FloatFromCurve(
     FloatCurve=DataInterface() {propertyValues {Curve {
        defaultValue=0
        keys [
         {arriveTangent=-0.051, interpMode=RCIM_Cubic, leaveTangent=-0.051, tangentMode=RCTM_User, time=0, value=1}
         {interpMode=RCIM_Cubic, tangentMode=RCTM_User, time=1, value=2}
        ]
       }}}
     CurveIndex=$Particles.Update Circuit Age
    )
    Module.Index0Float0Write=true
    Module.Index0Float1Write=true
    Module.Index0Float2Write=true
    Module.Index0Float3Write=true
   )
   &ScaleMeshSize(Scale Factor=&VectorFromFloat(Value=&FloatFromCurve(
      FloatCurve=DataInterface() {propertyValues {Curve {
         defaultValue=1
         keys [
          {arriveTangent=4.1785, interpMode=RCIM_Cubic, leaveTangent=4.1785, tangentMode=RCTM_User, time=0, value=1}
          {arriveTangent=1.1179, interpMode=RCIM_Cubic, leaveTangent=1.1179, tangentMode=RCTM_User, time=0.2892, value=2}
          {arriveTangent=0.0224, interpMode=RCIM_Cubic, leaveTangent=0.0224, tangentMode=RCTM_User, time=1, value=2.2}
         ]
        }}}
      CurveIndex=$Particles.Update Wave Age
     )))
   /SetVariables(
    Particles.Update Wave Color:LinearColor=&MultiplyLinearColor(
     Linear Color=$Particles.Wave Color
     Scale Factor=&MakeVector4FromVectorAndFloat(
      "Vector (RGB)"=&Multiply_VectorByFloat(
       Vector=<1 1 1>
       Float=&FloatFromCurve(
        FloatCurve=DataInterface() {propertyValues {Curve {keys [{arriveTangent=-1.7773, interpMode=RCIM_Cubic, leaveTangent=-1.7773, tangentMode=RCTM_User, time=0, value=1.718}, <at 1 1 RCIM_Cubic>]}}}
        CurveIndex=$Particles.Update Wave Age
        Scale Curve=2
       )
      )
      "Float (Alpha)"=&FloatFromCurve(
       FloatCurve=DataInterface() {propertyValues {Curve {keys [
           <at 0 0 RCIM_Cubic>
           {arriveTangent=0.5314, interpMode=RCIM_Cubic, leaveTangent=0.5314, time=0.1, value=1}
           {arriveTangent=-1.1111, interpMode=RCIM_Cubic, leaveTangent=-1.1111, time=0.591, value=0.3141}
           <at 1 0>
          ]}}}
       CurveIndex=$Particles.Update Wave Age
       Scale Curve=0.3
      )
     )
    )
    Particles.Update Wave Dynamic Parameter:Vector4f=&MakeVector4(
     X=&FloatFromCurve(FloatCurve=DataInterface() {propertyValues {Curve {keys [<at 0 1>,<at 1 0>]}}}, CurveIndex=$Particles.Update Wave Age)
     Y=2
     Z=&FloatFromCurve(FloatCurve=DataInterface() {propertyValues {Curve {keys [<at 0 1>,<at 1 0>]}}}, CurveIndex=$Particles.Update Wave Age)
     W=&FloatFromCurve(
      FloatCurve=DataInterface() {propertyValues {Curve {keys [
          <at 0 1>
          {arriveTangent=0.5265, interpMode=RCIM_Cubic, leaveTangent=0.5265, tangentMode=RCTM_User, time=0.3, value=0.8}
          <at 1 1 RCIM_Cubic>
         ]}}}
      CurveIndex=$Particles.Update Wave Age
     )
    )
    Particles.Update Wave Dynamic Parameter 001:Vector4f=&MakeVector4(X=40, Y=5, Z=0, W=0)
   ) {isSetParameters=true}
   /SetVariables(
    Particles.Update Glow Color:LinearColor=&MakeLinearColorFromVectorAndFloat(
     "Vector (RGB)"=&MakeVectorFromLinearColorRGB(Color=$Particles.MainColor)
     "Float (Alpha)"=&FloatFromCurve(
      FloatCurve=DataInterface() {propertyValues {Curve {keys [<at 0 0 RCIM_Cubic>,<at 0.3 1 RCIM_Cubic>,<at 1 0 RCIM_Cubic>]}}}
      CurveIndex=$Particles.Update Glow Age
      Scale Curve=2
     )
    )
    Particles.Update Glow Dynamic Parameter:Vector4f=<2 0 0 0>
    Particles.Update Glow Size:Vector2f=&Multiply_Vector2D(
     A=$Particles.Glow Size
     B=&Multiply_Vector2D_ByFloat(
      Float=&FloatFromCurve(
       FloatCurve=DataInterface() {propertyValues {Curve {
          defaultValue=1
          keys [
           {interpMode=RCIM_Cubic, tangentMode=RCTM_User, time=0, value=0.5}
           {arriveTangent=5.9147, interpMode=RCIM_Cubic, leaveTangent=5.9147, tangentMode=RCTM_User, time=1, value=3}
          ]
         }}}
       CurveIndex=$Particles.Update Glow Age
      )
      Vector2D=<1 1>
     )
    )
   ) {isSetParameters=true}
   /SetVariables(Particles.Update Dark Color:LinearColor=&MakeLinearColorFromVectorAndFloat(
     "Vector (RGB)"=&MakeVectorFromLinearColorRGB(Color=$Particles.Dark Color)
     "Float (Alpha)"=&FloatFromCurve(
      FloatCurve=DataInterface() {propertyValues {Curve {
         defaultValue=1
         keys [{interpMode=RCIM_Cubic, tangentMode=RCTM_User, time=0, value=1}, {interpMode=RCIM_Cubic, tangentMode=RCTM_User, time=1, value=0}]
        }}}
      CurveIndex=$Particles.Update Dark Age
      Scale Curve=0.3
     )
    )) {isSetParameters=true}
   /SetVariables(
    Particles.Update Background Color:LinearColor=&MakeLinearColorFromVectorAndFloat(
     "Vector (RGB)"=&MakeVectorFromLinearColorRGB(Color=$Particles.Dark Color)
     "Float (Alpha)"=&FloatFromCurve(
      FloatCurve=DataInterface() {propertyValues {Curve {
         defaultValue=1
         keys [{interpMode=RCIM_Cubic, tangentMode=RCTM_User, time=0, value=1}, {interpMode=RCIM_Cubic, tangentMode=RCTM_User, time=1, value=0}]
        }}}
      CurveIndex=$Particles.Update Dark Age
      Scale Curve=0.5
     )
    )
    Particles.Update Background Dynamic Parameter:Vector4f=<2 1 0.3 1>
    Particles.Update Background Size:Vector2f=&Multiply_Vector2D_ByFloat(
     Float=&FloatFromCurve(FloatCurve=DataInterface() {propertyValues {Curve {
         defaultValue=1
         keys [
          {arriveTangent=1.6106, interpMode=RCIM_Cubic, leaveTangent=1.6106, tangentMode=RCTM_User, time=0, value=0.5244}
          {interpMode=RCIM_Cubic, tangentMode=RCTM_User, time=1, value=1}
         ]
        }}})
     Vector2D=$Particles.Initial.Dark Size
    )
   ) {isSetParameters=true}
   &Light_Attributes(
    Color=1
    Color Value To Scale:LinearColor=$Particles.MainColor
    Enabled=2
    Light Enabled:NiagaraBool=&Boolean_NotOperation(Bool=&SetBoolByFloatComparison(A=$Particles.Update Circuit Age, B=0.3))
    Scale Alpha:NiagaraFloat=&FloatFromCurve(
     FloatCurve=DataInterface() {propertyValues {Curve {
        defaultValue=1
        keys [
         {interpMode=RCIM_Cubic, tangentMode=RCTM_User, time=0, value=0}
         {interpMode=RCIM_Cubic, tangentMode=RCTM_User, time=0.5, value=1}
         {tangentMode=RCTM_User, time=1, value=0}
        ]
       }}}
     Scale Curve=5
    )
    Scale RGB:Vector3f=<5 5 5>
    ScaleA=true
    ScaleRGB=true
   )
  ]
  rendererClasses [
   &NiagaraSpriteRendererProperties
   &NiagaraMeshRendererProperties
   &NiagaraLightRendererProperties
  ]
  renderers [
   {rendererClass=&NiagaraSpriteRendererProperties}
   {rendererClass=&NiagaraSpriteRendererProperties, rendererIndex=1}
   {rendererClass=&NiagaraMeshRendererProperties, rendererIndex=2}
   {rendererClass=&NiagaraLightRendererProperties, rendererIndex=3}
   {rendererClass=&NiagaraSpriteRendererProperties, rendererIndex=4}
  ]
 }}
summary {
 systemName=NS_Emote_FareSporeMilk_Circuit
 emitters [{
   emitterName=Circuitboard
   rendererClasses [
    &NiagaraSpriteRendererProperties
    &NiagaraMeshRendererProperties
    &NiagaraLightRendererProperties
   ]
  }]
 userVariables [
  User.AnimBeat {type=NiagaraFloat, defaultValue=0, description=""}
  User.AnimBeat2 {type=NiagaraFloat, defaultValue=0, description=""}
  User.bIsFrontEnd {type=NiagaraBool}
  User.bIsFrontEndPreview {type=NiagaraBool}
 ]
}
