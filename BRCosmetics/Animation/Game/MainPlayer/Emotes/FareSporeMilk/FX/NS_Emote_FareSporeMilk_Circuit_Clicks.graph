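# Niagara effect for the FareSporeMilk emote that flashes circuit clicks, energy waves and glows around its gadget prop.
# Circuitboard fires timed bursts on a roughly 0.46-second beat at two spawn-group offsets, cycling colors by emitter age and layering sprites and a mesh with separate age curves.
# EnergyWave_Burst adds delayed, color-cycling mesh waves.
# ScreenGlow pins five color-cycling sprites to the gadget's dial, button and knob sockets using partition masks.
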NiagaraSystem /BRCosmetics/Animation/Game/MainPlayer/Emotes/FareSporeMilk/FX/NS_Emote_FareSporeMilk_Circuit_Clicks
dependencies {
 usedDataInterfaces [
  &NiagaraDataInterfaceArrayFloat
  &NiagaraDataInterfaceArrayColor
  &NiagaraDataInterfaceArrayFloat3
  &NiagaraDataInterfaceColorCurve
  &NiagaraDataInterfaceCurve
  &NiagaraDataInterfaceSocketReader
 ]
 usedDynamicInputs [&Vector4FromCurve]
 usedModules [&SystemState]
}
emitters {
 Circuitboard {
  emitterUpdateScript [
   &EmitterState(Life Cycle Mode=::NewEnumerator1, Loop Behavior=::NewEnumerator1, Loop Delay=2.338, Loop Duration=10, UseLoopDelay=true)
   &SpawnBurst_Instantaneous(Spawn Group=1, Spawn Time=6, Use Spawn Probability=false, UseLoopCountLimit=false) {module=SpawnBurst_Instantaneous016}
   &SpawnBurst_Instantaneous(Spawn Group=1, Spawn Time=5.6, Use Spawn Probability=false, UseLoopCountLimit=false) {module=SpawnBurst_Instantaneous015}
   &SpawnBurst_Instantaneous(Spawn Group=1, Spawn Time=4.61, Use Spawn Probability=false, UseLoopCountLimit=false) {module=SpawnBurst_Instantaneous014}
   &SpawnBurst_Instantaneous(Spawn Group=1, Spawn Time=3.69, Use Spawn Probability=false, UseLoopCountLimit=false) {module=SpawnBurst_Instantaneous013}
   &SpawnBurst_Instantaneous(Spawn Group=1, Spawn Time=2.77, Use Spawn Probability=false, UseLoopCountLimit=false) {module=SpawnBurst_Instantaneous012}
   &SpawnBurst_Instantaneous(Spawn Time=5.53, Use Spawn Probability=false, UseLoopCountLimit=false) {module=SpawnBurst_Instantaneous017}
   &SpawnBurst_Instantaneous(Spawn Time=5.07, Use Spawn Probability=false, UseLoopCountLimit=false) {module=SpawnBurst_Instantaneous011}
   &SpawnBurst_Instantaneous(Spawn Time=4.61, Use Spawn Probability=false, UseLoopCountLimit=false) {module=SpawnBurst_Instantaneous010}
   &SpawnBurst_Instantaneous(Spawn Time=4.15, Use Spawn Probability=false, UseLoopCountLimit=false) {module=SpawnBurst_Instantaneous009}
   &SpawnBurst_Instantaneous(Spawn Time=3.69, Use Spawn Probability=false, UseLoopCountLimit=false) {module=SpawnBurst_Instantaneous008}
   &SpawnBurst_Instantaneous(Spawn Time=3.23, Use Spawn Probability=false, UseLoopCountLimit=false) {module=SpawnBurst_Instantaneous007}
   &SpawnBurst_Instantaneous(Spawn Time=2.77, Use Spawn Probability=false, UseLoopCountLimit=false) {module=SpawnBurst_Instantaneous006}
   &SpawnBurst_Instantaneous(Spawn Time=2.31, Use Spawn Probability=false, UseLoopCountLimit=false) {module=SpawnBurst_Instantaneous005}
   &SpawnBurst_Instantaneous(Spawn Time=1.85, Use Spawn Probability=false, UseLoopCountLimit=false) {module=SpawnBurst_Instantaneous004}
   &SpawnBurst_Instantaneous(Spawn Time=1.39, Use Spawn Probability=false, UseLoopCountLimit=false) {module=SpawnBurst_Instantaneous003}
   &SpawnBurst_Instantaneous(Spawn Time=0.91, Use Spawn Probability=false, UseLoopCountLimit=false) {module=SpawnBurst_Instantaneous002}
   &SpawnBurst_Instantaneous(Spawn Time=0.455, Use Spawn Probability=false, UseLoopCountLimit=false) {module=SpawnBurst_Instantaneous001}
   &SpawnBurst_Instantaneous(Use Spawn Probability=false, UseLoopCountLimit=false)
   /SetVariables(Emitter.Size:NiagaraFloat=&SelectFloatFromArray(
     Direct Array Index=&MakeIntFromFloat(Float=&Multiply_Float(A=$Emitter.Age, B=10))
     Interpolation=/Niagara/DynamicInputs/Math/Floor_Float(Float:NiagaraFloat=$Emitter.Age)
     Direct Array Mode=::NewEnumerator1
     Float Selection Array=DataInterface() {propertyValues {FloatData [0.25,0.35,0.25,0.37,0.22,0.3,0.33]}}
    )) {isSetParameters=true}
   /SetVariables(
    Emitter.CircuitColor:LinearColor=&SelectLinearColorFromArray(
     Array Sampling Mode=::NewEnumerator1
     Direct Array Index=&MakeIntFromFloat(Float=&Divide_Float(A=$Emitter.Age, B=0.46))
     Direct Array Mode=::NewEnumerator1
     Evaluation Type=::NewEnumerator1
     Interpolation=$Emitter.Age
     Color Selection Array=DataInterface() {propertyValues {ColorData [<rgba 0.7835 0.0924 1 1>, <rgba 1 0.7849 0.1747 1>, <rgba 0.0532 0.7534 1 1>]}}
     Random Seed=?
    )
    Emitter.CircuitColor001:LinearColor=&SelectLinearColorFromArray(
     Array Sampling Mode=::NewEnumerator1
     Direct Array Index=&MakeIntFromFloat(Float=&Divide_Float(A=$Emitter.Age, B=0.46))
     Direct Array Mode=::NewEnumerator1
     Evaluation Type=::NewEnumerator1
     Interpolation=$Emitter.Age
     Color Selection Array=DataInterface() {propertyValues {ColorData [<rgba 0.0532 0.7534 1 1>, <rgba 0.7835 0.0924 1 1>, <rgba 1 0.7849 0.1747 1>]}}
     Random Seed=?
    )
    Emitter.DarkBackgroundColor:LinearColor=&SelectLinearColorFromArray(
     Array Sampling Mode=::NewEnumerator1
     Direct Array Index=&MakeIntFromFloat(Float=&Divide_Float(A=$Emitter.Age, B=0.46))
     Direct Array Mode=::NewEnumerator1
     Evaluation Type=::NewEnumerator1
     Interpolation=$Emitter.Age
     Color Selection Array=DataInterface() {propertyValues {ColorData [<rgba 0.3182 0.0794 0.7587 1>, <rgba 0.8189 0.1122 0.4814 5>, <rgba 0.1224 0.3202 0.8652 5>]}}
     Random Seed=?
    )
    Emitter.DarkBackgroundColor001:LinearColor=&SelectLinearColorFromArray(
     Array Sampling Mode=::NewEnumerator1
     Direct Array Index=&MakeIntFromFloat(Float=&Divide_Float(A=$Emitter.Age, B=0.46))
     Direct Array Mode=::NewEnumerator1
     Evaluation Type=::NewEnumerator1
     Interpolation=$Emitter.Age
     Color Selection Array=DataInterface() {propertyValues {ColorData [<rgba 0.0664 0.5617 0.8652 5>, <rgba 0.3182 0.0794 0.7587 1>, <rgba 0.8189 0.1122 0.4814 5>]}}
     Random Seed=?
    )
    Emitter.WaveColor:LinearColor=&SelectLinearColorFromArray(
     Array Sampling Mode=::NewEnumerator1
     Direct Array Index=&MakeIntFromFloat(Float=&Divide_Float(A=$Emitter.Age, B=0.46))
     Direct Array Mode=::NewEnumerator1
     Evaluation Type=::NewEnumerator1
     Interpolation=$Emitter.Age
     Color Selection Array=DataInterface() {propertyValues {ColorData [<rgba 0.3556 0 0.467 1>, <rgba 0.1334 0.0055 0.1753 1>, <rgba 0.7622 0.0518 1 1>]}}
     Random Seed=?
    )
    Emitter.WaveColor001:LinearColor=&SelectLinearColorFromArray(
     Array Sampling Mode=::NewEnumerator1
     Direct Array Index=&MakeIntFromFloat(Float=&Divide_Float(A=$Emitter.Age, B=0.46))
     Direct Array Mode=::NewEnumerator1
     Evaluation Type=::NewEnumerator1
     Interpolation=$Emitter.Age
     Color Selection Array=DataInterface() {propertyValues {ColorData [<rgba 0.7622 0.0518 1 1>, <rgba 0.3556 0 0.467 1>, <rgba 0.1334 0.0055 0.1753 1>]}}
     Random Seed=?
    )
   ) {isSetParameters=true}
  ]
  particleSpawnScript [
   &InitializeParticle(
    Color=<rgba 10 10 10 1>
    Lifetime=0.5
    Mesh Scale Mode=::NewEnumerator3
    Mesh Uniform Scale=&Multiply_Float(A=0.5, B=$Emitter.Size)
    Sprite Rotation Angle=&SelectFloatFromArray(
     Array Sampling Mode=::NewEnumerator0
     Direct Array Index=?
     Float Selection Array=DataInterface() {propertyValues {FloatData [0,0.25,0.5,0.75]}}
     Random Seed=0
    )
    Sprite Rotation Mode=::NewEnumerator1
    Sprite Size Mode=::NewEnumerator3
    Uniform Sprite Size=&Multiply_Float(A=157, B=$Emitter.Size)
    UsePositionOffset=true
    Write Material Random=true
   )
   /SetVariables(
    Particles.Wave Color:LinearColor=&Multiply_LinearColorByFloat(
     Linear Color=&HueShiftLinearColor(
      Hue=0.1422
      Saturation=5.3578
      Value=1
      Linear Color=&MaskLinearColorBySpawnGroup(Masked Linear Color=$Emitter.WaveColor, Passthrough Linear Color=$Emitter.WaveColor001)
     )
     Float=2
    )
    Particles.Wave Lifetime=0.5
   ) {isSetParameters=true}
   /SetVariables(
    Particles.Glow Lifetime=0.2
    Particles.Glow Size=&Multiply_Vector2D_ByFloat(Float=$Emitter.Size, Vector2D=&Multiply_Vector2D_ByFloat(Float=40, Vector2D=<1 1>))
   ) {isSetParameters=true}
   /SetVariables(Particles.Circuit LIfetime=0.5) {isSetParameters=true}
   /SetVariables(
    Particles.Dark Color:LinearColor=&MaskLinearColorBySpawnGroup(Masked Linear Color=$Emitter.DarkBackgroundColor, Passthrough Linear Color=$Emitter.DarkBackgroundColor001)
    Particles.Dark Lifetime=0.4
    Particles.Dark Size:Vector2f=<70 70>
   ) {isSetParameters=true}
   &InitialMeshOrientation(Mesh Orientation Mode=::NewEnumerator2, Orientation Vector=<0 0 1>, Use Rotation Vector=false)
   &CameraOffset(Camera Offset Amount=5)
   &SystemLocation(
    MaskSpawnGroup=false
    Offset=&SelectVectorFromArray(
     Direct Array Index=&SpawnGroup()
     Direct Array Mode=::NewEnumerator1
     Vector Selection Array=DataInterface() {propertyValues {FloatData [<70 -6 15>, <55 38 16>]}}
    )
   )
  ]
  particleUpdateScript [
   &ParticleState()
   /SetVariables(
    Particles.Update Circuit Age:NiagaraFloat=&Divide_Float(A=$Particles.Age, B=$Particles.Circuit LIfetime)
    Particles.Update Dark Age:NiagaraFloat=&Divide_Float(A=$Particles.Age, B=$Particles.Dark Lifetime)
    Particles.Update Glow Age:NiagaraFloat=&Divide_Float(A=$Particles.Age, B=$Particles.Glow Lifetime)
    Particles.Update Wave Age:NiagaraFloat=&Divide_Float(A=$Particles.Age, B=$Particles.Wave Lifetime)
   ) {isSetParameters=true}
   &ScaleColor(
    Scale Alpha=&Multiply_Float(
     A=&FNTimeOfDayScalar(Night Scale=0.5, Morning Scale=0.5, Evening Scale=0.4)
     B=&FloatFromCurve(FloatCurve=DataInterface() {propertyValues {Curve {keys [<at 0 0>,<at 0.1 1>,<at 0.8 1>,<at 1 0>]}}}, Scale Curve=5)
    )
    Scale RGB=&MakeVectorFromLinearColorRGB(Color=&Multiply_LinearColorByFloat(
      Linear Color=&MaskLinearColorBySpawnGroup(Masked Linear Color=$Emitter.CircuitColor, Passthrough Linear Color=$Emitter.CircuitColor001)
      Float=3
     ))
    ScaleA=true
    ScaleRGB=true
   )
   &DynamicMaterialParameters(
    Index 0 Param 1=&FloatFromCurve(
     FloatCurve=DataInterface() {propertyValues {Curve {
        defaultValue=1
        keys [{arriveTangent=2.7631, interpMode=RCIM_Cubic, leaveTangent=2.7631, tangentMode=RCTM_User, time=0, value=0}, <at 0.9013 1 RCIM_Cubic>]
       }}}
     CurveIndex=$Particles.Update Circuit Age
    )
    Index 0 Param 2=&FloatFromCurve(
     FloatCurve=DataInterface() {propertyValues {Curve {
        defaultValue=0
        keys [{interpMode=RCIM_Cubic, tangentMode=RCTM_User, time=0, value=1}, {interpMode=RCIM_Cubic, tangentMode=RCTM_User, time=1, value=2}]
       }}}
     CurveIndex=$Particles.Update Circuit Age
    )
    Module.Index0Float0Write=true
    Module.Index0Float1Write=true
    Module.Index0Float2Write=true
    Module.Index0Float3Write=true
   )
   &ScaleMeshSize(Scale Factor=&VectorFromFloat(Value=&FloatFromCurve(
      FloatCurve=DataInterface() {propertyValues {Curve {
         defaultValue=1
         keys [
          {arriveTangent=4.1785, interpMode=RCIM_Cubic, leaveTangent=4.1785, tangentMode=RCTM_User, time=0, value=1}
          {arriveTangent=1.1179, interpMode=RCIM_Cubic, leaveTangent=1.1179, tangentMode=RCTM_User, time=0.2892, value=2}
          {arriveTangent=0.0224, interpMode=RCIM_Cubic, leaveTangent=0.0224, tangentMode=RCTM_User, time=1, value=2.2}
         ]
        }}}
      CurveIndex=$Particles.Update Wave Age
     )))
   /SetVariables(
    Particles.Update Wave Color:LinearColor=&MultiplyLinearColor(
     Linear Color=$Particles.Wave Color
     Scale Factor=&MakeVector4FromVectorAndFloat(
      "Vector (RGB)"=&Multiply_VectorByFloat(
       Vector=<1 1 1>
       Float=&FloatFromCurve(
        FloatCurve=DataInterface() {propertyValues {Curve {keys [{arriveTangent=-1.7773, interpMode=RCIM_Cubic, leaveTangent=-1.7773, tangentMode=RCTM_User, time=0, value=1.718}, <at 1 1 RCIM_Cubic>]}}}
        CurveIndex=$Particles.Update Wave Age
       )
      )
      "Float (Alpha)"=&FloatFromCurve(
       FloatCurve=DataInterface() {propertyValues {Curve {keys [
           <at 0 0 RCIM_Cubic>
           {arriveTangent=0.5314, interpMode=RCIM_Cubic, leaveTangent=0.5314, time=0.1, value=1}
           {arriveTangent=-1.1111, interpMode=RCIM_Cubic, leaveTangent=-1.1111, time=0.591, value=0.3141}
           <at 1 0>
          ]}}}
       CurveIndex=$Particles.Update Wave Age
       Scale Curve=0.2
      )
     )
    )
    Particles.Update Wave Dynamic Parameter:Vector4f=&MakeVector4(
     X=&FloatFromCurve(FloatCurve=DataInterface() {propertyValues {Curve {keys [<at 0 1>,<at 1 0>]}}}, CurveIndex=$Particles.Update Wave Age)
     Y=2
     Z=&FloatFromCurve(FloatCurve=DataInterface() {propertyValues {Curve {keys [<at 0 1>,<at 1 0>]}}}, CurveIndex=$Particles.Update Wave Age)
     W=&FloatFromCurve(
      FloatCurve=DataInterface() {propertyValues {Curve {keys [
          <at 0 1>
          {arriveTangent=0.5265, interpMode=RCIM_Cubic, leaveTangent=0.5265, tangentMode=RCTM_User, time=0.3, value=0.8}
          <at 1 1 RCIM_Cubic>
         ]}}}
      CurveIndex=$Particles.Update Wave Age
     )
    )
    Particles.Update Wave Dynamic Parameter 001:Vector4f=&MakeVector4(X=40, Y=5, Z=0, W=0)
   ) {isSetParameters=true}
   /SetVariables(
    Particles.Update Glow Color:LinearColor=&MakeLinearColorFromVectorAndFloat(
     "Vector (RGB)"=&MakeVectorFromLinearColorRGB(Color=&MaskLinearColorBySpawnGroup(Masked Linear Color=$Emitter.CircuitColor, Passthrough Linear Color=$Emitter.CircuitColor001))
     "Float (Alpha)"=&FloatFromCurve(
      FloatCurve=DataInterface() {propertyValues {Curve {keys [<at 0 0 RCIM_Cubic>,<at 0.3 1 RCIM_Cubic>,<at 1 0 RCIM_Cubic>]}}}
      CurveIndex=$Particles.Update Glow Age
      Scale Curve=2
     )
    )
    Particles.Update Glow Dynamic Parameter:Vector4f=<2 0 0 0>
    Particles.Update Glow Size:Vector2f=&Multiply_Vector2D(
     A=$Particles.Glow Size
     B=&Multiply_Vector2D_ByFloat(
      Float=&FloatFromCurve(
       FloatCurve=DataInterface() {propertyValues {Curve {
          defaultValue=1
          keys [{interpMode=RCIM_Cubic, tangentMode=RCTM_User, time=0, value=0.5}, {interpMode=RCIM_Cubic, tangentMode=RCTM_User, time=1, value=3}]
         }}}
       CurveIndex=$Particles.Update Glow Age
      )
      Vector2D=<1 1>
     )
    )
   ) {isSetParameters=true}
   /SetVariables(Particles.Update Dark Color:LinearColor=&MakeLinearColorFromVectorAndFloat(
     "Vector (RGB)"=&MakeVectorFromLinearColorRGB(Color=$Particles.Dark Color)
     "Float (Alpha)"=&FloatFromCurve(
      FloatCurve=DataInterface() {propertyValues {Curve {
         defaultValue=1
         keys [{interpMode=RCIM_Cubic, tangentMode=RCTM_User, time=0, value=1}, {interpMode=RCIM_Cubic, tangentMode=RCTM_User, time=1, value=0}]
        }}}
      CurveIndex=$Particles.Update Dark Age
      Scale Curve=0.2
     )
    )) {isSetParameters=true}
   /SetVariables(
    Particles.Update Background Color:LinearColor=&MakeLinearColorFromVectorAndFloat(
     "Vector (RGB)"=&MakeVectorFromLinearColorRGB(Color=$Particles.Dark Color)
     "Float (Alpha)"=&FloatFromCurve(
      FloatCurve=DataInterface() {propertyValues {Curve {
         defaultValue=1
         keys [{interpMode=RCIM_Cubic, tangentMode=RCTM_User, time=0, value=1}, {interpMode=RCIM_Cubic, tangentMode=RCTM_User, time=1, value=0}]
        }}}
      CurveIndex=$Particles.Update Dark Age
     )
    )
    Particles.Update Background Dynamic Parameter:Vector4f=<3 1 1 1>
    Particles.Update Background Size:Vector2f=&Multiply_Vector2D_ByFloat(
     Float=&FloatFromCurve(FloatCurve=DataInterface() {propertyValues {Curve {
         defaultValue=1
         keys [
          {arriveTangent=1.6106, interpMode=RCIM_Cubic, leaveTangent=1.6106, tangentMode=RCTM_User, time=0, value=0.5244}
          {interpMode=RCIM_Cubic, tangentMode=RCTM_User, time=1, value=1}
         ]
        }}})
     Vector2D=$Particles.Initial.Dark Size
    )
   ) {isSetParameters=true}
  ]
  rendererClasses [&NiagaraSpriteRendererProperties, &NiagaraMeshRendererProperties]
  renderers [
   {rendererClass=&NiagaraSpriteRendererProperties}
   {rendererClass=&NiagaraSpriteRendererProperties, rendererIndex=1}
   {rendererClass=&NiagaraMeshRendererProperties, rendererIndex=2}
   {rendererClass=&NiagaraSpriteRendererProperties, rendererIndex=3}
  ]
 }
 EnergyWave_Burst {
  emitterUpdateScript [
   &EmitterState(Life Cycle Mode=::NewEnumerator1, Loop Behavior=::NewEnumerator1, Loop Delay=4.2, Loop Duration=6, UseLoopDelay=true)
   &SpawnBurst_Instantaneous(Spawn Time=2, Use Spawn Probability=false, UseLoopCountLimit=false) {module=SpawnBurst_Instantaneous001}
   &SpawnBurst_Instantaneous(Use Spawn Probability=false, UseLoopCountLimit=false)
  ]
  particleSpawnScript [&InitializeParticle(
    Lifetime=4
    Mesh Scale Mode=::NewEnumerator3
    Mesh Uniform Scale=2.2
    Position Offset=<0 0 -5>
    UsePositionOffset=true
    Write Material Random=true
   )]
  particleUpdateScript [
   &ParticleState()
   &ScaleMeshSize(Scale Factor=&VectorFromFloat(Value=&FloatFromCurve(FloatCurve=DataInterface() {propertyValues {Curve {keys [<at 0 0.6131 RCIM_Cubic>,<at 1 1 RCIM_Cubic>]}}}, Scale Curve=0.9)))
   &ScaleColor(
    Scale Alpha=&Multiply_Float(
     A=&FloatFromCurve(FloatCurve=DataInterface() {propertyValues {Curve {
         defaultValue=1
         keys [
          {arriveTangent=1.8488, interpMode=RCIM_Cubic, leaveTangent=1.8488, tangentMode=RCTM_User, time=0, value=0.7742}
          {interpMode=RCIM_Cubic, tangentMode=RCTM_User, time=0.8664, value=1}
          {arriveTangent=-0.0844, interpMode=RCIM_Cubic, leaveTangent=-0.0844, tangentMode=RCTM_User, time=1, value=0}
         ]
        }}})
     B=&FloatFromCurve(
      FloatCurve=DataInterface() {propertyValues {Curve {
         defaultValue=1
         keys [
          {arriveTangent=1.8488, interpMode=RCIM_Cubic, leaveTangent=1.8488, tangentMode=RCTM_User, time=0, value=0.1095}
          {arriveTangent=1.2205, interpMode=RCIM_Cubic, leaveTangent=1.2205, time=0.1, value=0.95}
          {interpMode=RCIM_Cubic, tangentMode=RCTM_User, time=0.7296, value=1}
          {arriveTangent=-0.0844, interpMode=RCIM_Cubic, leaveTangent=-0.0844, tangentMode=RCTM_User, time=1, value=0}
         ]
        }}}
      CurveIndex=$Emitter.NormalizedLoopAge
     )
    )
    Scale RGB=&MakeVectorFromLinearColorRGB(Color=&ColorFromCurve(
      ColorCurve=DataInterface() {propertyValues {
        AlphaCurve {keys [<at 0 1>]}
        BlueCurve {keys [<at 0 0.5701 RCIM_Cubic>,<at 0.8882 2>]}
        GreenCurve {keys [<at 0 0 RCIM_Cubic>,<at 0.8882 0>]}
        RedCurve {keys [<at 0 4 RCIM_Cubic>,<at 0.8882 0.209>]}
       }}
      CurveIndex=&Frac_Float(Float=&Multiply_Float(A=$Particles.NormalizedAge, B=8))
     ))
    ScaleA=true
    ScaleRGB=true
   )
   &DynamicMaterialParameters(
    Index 0 Param 1=&FloatFromCurve(FloatCurve=DataInterface() {propertyValues {Curve {keys [<at 0 1>,<at 0.2 1.375>,<at 1 0>]}}}, Scale Curve=1.2)
    Index 0 Param 2=1.3
    Index 0 Param 3=&FloatFromCurve(FloatCurve=DataInterface() {propertyValues {Curve {keys [<at 0.1 1>,<at 1 0>]}}})
    Index 0 Param 4=&FloatFromCurve(FloatCurve=DataInterface() {propertyValues {Curve {keys [
         <at 0 1>
         {arriveTangent=0.3079, interpMode=RCIM_Cubic, leaveTangent=0.3079, tangentMode=RCTM_User, time=0.3, value=0.9}
         <at 1 1 RCIM_Cubic>
        ]}}})
    Index 1 Param 1=100
    Index 1 Param 2=100
    Module.Index0Float0Write=true
    Module.Index0Float1Write=true
    Module.Index0Float2Write=true
    Module.Index0Float3Write=true
    Module.Index1Float0Write=true
    Module.Index1Float1Write=true
    Module.Index1Float2Write=true
    Module.Index1Float3Write=true
    Write Parameter Index 1=::NewEnumerator1
   )
  ]
  rendererClasses [&NiagaraMeshRendererProperties]
  renderers [{rendererClass=&NiagaraMeshRendererProperties, rendererIndex=0}]
 }
 ScreenGlow {
  emitterUpdateScript [
   &EmitterState(
    Inactive Response=::NewEnumerator1
    Life Cycle Mode=::NewEnumerator1
    Loop Behavior=::NewEnumerator1
    Loop Delay=0.5
    Loop Duration Mode=::NewEnumerator1
    UseLoopDelay=false
   )
   &SpawnBurst_Instantaneous(Spawn Count=5, Use Spawn Probability=false)
  ]
  particleSpawnScript [
   &PartitionParticles(Number Of Partitions=5)
   &InitializeParticle(
    Color=<rgba 0.4323 0.4323 0.4323 0.2>
    Sprite Size=/Niagara/DynamicInputs/Helpers/MaskVector2d(
     Mask Source:NiagaraInt32=1
     Mask Parameter:NiagaraInt32=$Particles.PartitionParticles.Partition
     Mask:NiagaraInt32=0
     Masked Vector2D:Vector2f=<80 60>
     Passthrough Vector2D:Vector2f=&MaskVector2d(
      Mask Source:NiagaraInt32=1
      Mask Parameter:NiagaraInt32=$Particles.PartitionParticles.Partition
      Mask:NiagaraInt32=1
      Masked Vector2D:Vector2f=<40 30>
      Passthrough Vector2D:Vector2f=&MaskVector2d(
       Mask Source:NiagaraInt32=1
       Mask Parameter:NiagaraInt32=$Particles.PartitionParticles.Partition
       Mask:NiagaraInt32=2
       Masked Vector2D:Vector2f=<40 30>
       Passthrough Vector2D:Vector2f=&MaskVector2d(
        Mask Source:NiagaraInt32=1
        Mask Parameter:NiagaraInt32=$Particles.PartitionParticles.Partition
        Mask:NiagaraInt32=3
        Masked Vector2D:Vector2f=<50 50>
        Passthrough Vector2D:Vector2f=<70 70>
       )
      )
     )
    )
    Sprite Size Mode=::NewEnumerator4
    UsePositionOffset=false
    Write Material Random=true
   ) {module=InitializeParticle001}
   &DynamicMaterialParameters(
    Index 0 Param 1=2
    Module.Param0WriteEnabled=true
    Module.Param1WriteEnabled=true
    Module.Param2WriteEnabled=true
    Module.Param3WriteEnabled=true
   )
   &CameraOffset()
  ]
  particleUpdateScript [
   &ParticleState(Kill Particles When Lifetime Has Elapsed=false, Let Infinitely Lived Particles Die When Emitter Deactivates=true)
   &SocketLocation(
    SamplePosOffset=true
    Sampled Position Offset=&MaskVector(
     Mask Source=1
     Mask Parameter=$Particles.PartitionParticles.Partition
     Mask=0
     Masked Vector=<0 0 6>
     Passthrough Vector=&MaskVector(Mask Source=1, Mask Parameter=$Particles.PartitionParticles.Partition, Mask=4, Masked Vector=<0 0 6>, Passthrough Vector=<0 0 0>)
    )
    Socket Index=$Particles.PartitionParticles.Partition
    Socket Reader=DataInterface() {propertyValues {
      AttachComponentClass=null
      AttachComponentTag=None
      EditorPreviewAsset=/BRCosmetics/Animation/Props/Gadgets_FareSporeMilk/Meshes/FareSporeMilk_Gadget
      SourceActorRef=null
      SourceAsset=&FareSporeMilk_Gadget
      SourceMode=Default
      bRequireCurrentFrameData=true
      bUpdateSocketsPerFrame=true
      FilteredSockets [Dials_Etc_JNT, Beat_Buttons_2_JNT, Beat_Buttons_1_JNT, Beat_Buttons_BIG_JNT, Knobs_JNT]
      ObjectParameterBinding {parameter {name=None, typeDefHandle {registeredTypeIndex=1}, varData []}}
     }}
    Socket Sampling Mode=::NewEnumerator3
    VelocitySpeedLimit=false
   )
   &SpriteFacingAndAlignment(
    Sprite Alignment=<0 0 1>
    Sprite Facing=&NormalizeVector(
     Vector To Normalize=&SubtractVectorFromPosition(
      Position=&AddVectorToPosition(Position=$Engine.Emitter.SimulationPosition, Vector=<0 0 60>)
      Vector=Linked() {linkedVariable {name=Particles.Position, type=NiagaraPosition}}
     )
     Safe Normalize=true
     Safe Fallback Threshold=0
     Safe Fallback Vector=<0 0 0.001>
    )
    WriteAlignment=true
    WriteFacing=true
   )
   &ScaleSpriteSize(Scale Factor=&Vector2DFromFloat(Value=&FloatFromCurve(FloatCurve=DataInterface() {propertyValues {Curve {keys [<at 0 1 RCIM_Cubic>]}}})))
   &ScaleColor(
    Scale Alpha=&FloatFromCurve(FloatCurve=DataInterface() {propertyValues {Curve {
        defaultValue=0
        keys [{interpMode=RCIM_Cubic, tangentMode=RCTM_User, time=0, value=0}, {interpMode=RCIM_Cubic, tangentMode=RCTM_User, time=0.3, value=1}]
       }}})
    Scale RGB=&MakeVectorFromLinearColorRGB(Color=&ColorFromCurve(
      ColorCurve=DataInterface() {propertyValues {
        AlphaCurve {keys [<at 0 1 RCIM_Cubic>]}
        BlueCurve {keys [
          <at 0 1 RCIM_Cubic>
          <at 0.368 0.6627>
          {arriveTangent=0.5337, interpMode=RCIM_Cubic, leaveTangent=0.5337, time=0.6974, value=0.195}
          <at 1 1>
         ]}
        GreenCurve {keys [
          <at 0 0.8345 RCIM_Cubic>
          <at 0.368 0.3015>
          {arriveTangent=0.8433, interpMode=RCIM_Cubic, leaveTangent=0.8433, time=0.6974, value=0.7185}
          <at 1 0.8345>
         ]}
        RedCurve {keys [
          <at 0 0.3617 RCIM_Cubic>
          <at 0.368 1>
          {arriveTangent=-1.0099, interpMode=RCIM_Cubic, leaveTangent=-1.0099, time=0.6974, value=1}
          <at 1 0.3617>
         ]}
       }}
      CurveIndex=&Frac_Float(Float=&Add_Float(
        A=&Multiply_Float(A=$Emitter.Age, B=0.2)
        B=&RandomRangeFloat(Minimum=0, Maximum=1)
       ))
     ))
    ScaleA=true
    ScaleRGB=true
   ) {module=ScaleColor001}
  ]
  rendererClasses [&NiagaraSpriteRendererProperties]
  renderers [{rendererClass=&NiagaraSpriteRendererProperties}]
 }
}
summary {
 systemName=NS_Emote_FareSporeMilk_Circuit_Clicks
 emitters [
  {emitterName=Circuitboard, rendererClasses [&NiagaraSpriteRendererProperties, &NiagaraMeshRendererProperties]}
  {emitterName=ScreenGlow, rendererClasses [&NiagaraSpriteRendererProperties]}
  {emitterName=EnergyWave_Burst, rendererClasses [&NiagaraMeshRendererProperties]}
 ]
 userVariables [
  User.bIsFrontEnd {type=NiagaraBool}
  User.bIsFrontEndPreview {type=NiagaraBool}
 ]
}
